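# Conversation database for the NPC Zadie, giving her vault-curious greeting lines and access to shared NPC conversation branches.
# Its entry tag leads to one Speech node with a default message, a first-ever-conversation line and four variants with empty requirement lists.
# Contextual lines take priority from named curve rows; the four variants use the OutfitReaction row with a null curve table.
# The speech node then links to the shared NPCLibrary Quests and BasicServices entry points.
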ConversationDatabase /NPCLibrary/NPCs/Zadie/NPCLibrary_Conversation_Zadie
entry_points [{tag=Conversation.Entry.Zadie, destinations [#1]}]
nodes [
 ConversationEntryPointNode_7 {
  id=#1
  class=ConversationEntryPointNode
  outputs [#2]
  properties {entryTag {tagName=Conversation.Entry.Zadie}}
 }
 Speech_C_4 {
  id=#2
  class=Speech_C
  outputs [#3, #4]
  properties {
   nodeName=""
   generalConfig {
    defaultMessage=Mask on. Masq ready. Let's have some fun.
    contextualMessages [
     {
      message=What do you think they keep in those fancy vaults? A bespoke gold toothpick collection?
      contextRequirements [{requirement {class=FortControllerRequirement_IsFirstEverConversationWithNPC}}]
      priority {curve {rowName=Tandem.Speech.Priority.FirstEverInteraction}}
     }
     {
      message=Mask on. Masq ready. Let's have some fun.
      requirementMatchPolicy=RequireAny
      contextRequirements []
      priority {curve {curveTable=null, rowName=Tandem.Speech.Priority.OutfitReaction}}
     }
     {
      message=I just wanna know what's in those vaults...
      requirementMatchPolicy=RequireAny
      contextRequirements []
      priority {curve {curveTable=null, rowName=Tandem.Speech.Priority.OutfitReaction}}
     }
     {
      message="Cracking vaults is just like opening a big metal present!"
      requirementMatchPolicy=RequireAny
      contextRequirements []
      priority {curve {curveTable=null, rowName=Tandem.Speech.Priority.OutfitReaction}}
     }
     {
      message=Some people say I'm too curious for my own good.
      requirementMatchPolicy=RequireAny
      contextRequirements []
      priority {curve {curveTable=null, rowName=Tandem.Speech.Priority.OutfitReaction}}
     }
    ]
   }
  }
 }
 ConversationLinkNode_0 {id=#4, class=ConversationLinkNode, properties {remoteEntryTag {tagName=Conversation.Entry.BasicServices}}}
 ConversationLinkNode_1 {id=#3, class=ConversationLinkNode, properties {remoteEntryTag {tagName=Conversation.Entry.NPCLibrary.Quests}}}
]
